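# Sound cue for the Figment Burst SMG's bullet whiz-by, built from Burst Fire SMG bullet-by waves in the Athena weapon sounds.
# A Random node picks among six equally weighted waves and feeds a Modulator that varies pitch and volume.
# An attenuation override turns off distance volume falloff and sets a wide stereo spread.
# It uses the ranged-projectile effects sound class and whiz-by concurrency rules, with virtualization disabled.
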SoundCue /Figment_Burst_SMG/Sounds/SC_Pistol_BurstFireSMG_Bullet_Whip_Figment_Cue
root=Modulator_0
nodes [
 {id=Modulator_0, children [Random_2], properties {pitchMax=1.15, volumeMax=1, volumeMin=0.6}}
 {
  id=Random_2
  children [WavePlayer_5, WavePlayer_7, WavePlayer_8, WavePlayer_9, WavePlayer_10, WavePlayer_11]
  properties {weights [1,1,1,1,1,1]}
 }
 {id=WavePlayer_10, properties {soundWaveAssetPtr=/Game/Athena/Sounds/Weapons/Pistol_BurstFireSMG/Pistol_BurstFireSMG_Bullet_By_14}}
 {id=WavePlayer_11, properties {soundWaveAssetPtr=/Game/Athena/Sounds/Weapons/Pistol_BurstFireSMG/Pistol_BurstFireSMG_Bullet_By_07}}
 {id=WavePlayer_5, properties {soundWaveAssetPtr=/Game/Athena/Sounds/Weapons/Pistol_BurstFireSMG/Pistol_BurstFireSMG_Bullet_By_09}}
 {id=WavePlayer_7, properties {soundWaveAssetPtr=/Game/Athena/Sounds/Weapons/Pistol_BurstFireSMG/Pistol_BurstFireSMG_Bullet_By_11}}
 {id=WavePlayer_8, properties {soundWaveAssetPtr=/Game/Athena/Sounds/Weapons/Pistol_BurstFireSMG/Pistol_BurstFireSMG_Bullet_By_12}}
 {id=WavePlayer_9, properties {soundWaveAssetPtr=/Game/Athena/Sounds/Weapons/Pistol_BurstFireSMG/Pistol_BurstFireSMG_Bullet_By_13}}
]
properties {
 bOverrideAttenuation=true
 maxDistance=2097152
 priority=0.5
 soundClassObject=&Fort_WeaponRanged_Projectile_Effects
 virtualizationMode=Disabled
 volumeMultiplier=0.5
 attenuationOverrides {bAttenuate=false, stereoSpread=5000}
 concurrencySet [/Game/Sounds/Concurrency_Rules/Athena_BulletWhizBy_Conc]
}
