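# Sound cue for the Figment Burst SMG's low-ammo sound, named after the burst-fire pistol SMG.
# A single wave player reuses the Athena burst-fire SMG mid-reload sample, with raised pitch and volume.
# It uses the shared weapon reload attenuation and a ranged-weapon low-ammo sound class.
# Its concurrency override prevents new instances from starting once the limit is reached, rather than cutting off playing ones.
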SoundCue /Figment_Burst_SMG/Sounds/SC_Pistol_BurstFireSMG_Low_Ammo_Figment
root=WavePlayer_0
nodes [{id=WavePlayer_0, properties {soundWaveAssetPtr=/Game/Athena/Sounds/Weapons/Pistol_BurstFireSMG/Pistol_BurstFireSMG_Reload_Mid_01}}]
properties {
 attenuationSettings=&Fort_Weapon_Reload_Att
 bOverrideConcurrency=true
 maxDistance=2250
 pitchMultiplier=1.2
 soundClassObject=/Game/Sounds/FortSoundClassesAndModes/Fort_WeaponRanged_LowAmmo_Degraded_PSM
 volumeMultiplier=1.5
 concurrencyOverrides {resolutionRule=PreventNew}
}
